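; BP_Portal_Glow is the portal glow actor for GrillMoon Countdown2, which removes itself when players teleport up into the air.
; On BeginPlay it waits for playlist data, then listens to the portal glow stateful event on the global gameplay event router.
; Both live events and saved event state are handled, so the actor destroys itself whenever the payload flag is false.
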(event EventBeginPlay
  (Class|FortAsyncActionCurrentPlaylistReady|PlaylistDataReadyAsync
    (:OnReady
      (bind (_asynctaskproxy _payload) (GameplayEventRouter|StartListeningToStatefulEvent (GameplayEventRouter|GetGlobalEventRouter) "/GrillMoon_Countdown2/Gameplay/EventRouters/GrillMoon_PortalGlow_EventRouter.GrillMoon_PortalGlow_EventRouter")
        (:OnEventReceived
          _)
        (:OnSavedEventStateExists
          _))
      (if (not (Utilities|Struct|BreakGrillMoonPortalGlowEventRouter _payload))
        (Actor|DestroyActor) ; turn off when players\nteleport up into the air
      ))))

(fn ConstructionScript)
